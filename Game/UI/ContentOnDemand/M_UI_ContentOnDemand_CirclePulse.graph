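# Translucent UI material for Content On Demand that computes a looping circular ping of two expanding, fading rings.
# Time and PingSpeed drive a looping phase, flippable by Reverse and shaped by two different power curves so the rings separate.
# Each curve drives MF_UI_Ring radius, thickness and softness.
# The rings combine by Max, scaled by PingAmount, into a Pulse reroute, with no material output connection shown.
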Material /Game/UI/ContentOnDemand/M_UI_ContentOnDemand_CirclePulse
graph {
 properties {
  bUsedWithStaticMesh=false
  blendMode=BLEND_Translucent
  materialDomain=MD_UI
  pixelDepthOffsetMode=PDOM_Legacy
  usedShadingModels=Default Lit
 }
 nodes [
  n1 = ScalarParameter() {defaultValue=1, group=Ping, parameterName=PingSpeed}
  n2 = Fmod(A=Multiply(A=Time() {period=1}, B=@n1) {constB=0.5}, B=Constant() {r=1})
  n3 = ScalarParameter() {parameterName=Reverse}
  n4 = LinearInterpolate(A=@n2, B=OneMinus(None=@n2), Alpha=@n3)
  Add(A=Multiply(A=@n4) {constB=0.5}) {constB=0.5}
  Constant() {r=1}
  n5 = ScalarParameter() {defaultValue=0.1, group=Ping, parameterName=PingRingMaxSize}
  n6 = Power(Base=@n4) {constExponent=0.4}
  n7 = LinearInterpolate(A=@n5, Alpha=@n6) {constA=0.1, constB=0.01}
  n8 = LinearInterpolate(Alpha=@n6) {constA=0.002, constB=0.05}
  n9 = ScalarParameter() {defaultValue=0.5, group=Ping, parameterName=PingOuterRadius}
  n10 = LinearInterpolate(B=@n9, Alpha=@n6) {constB=0.5}
  n11 = Clamp(None=Multiply(
    A=MaterialFunctionCall!(RingSize=@n7, Edge, EdgeSoftness=@n8, OuterRadius=@n10) {materialFunction=/Game/UI/InGame/HUD/Textures/MF_UI_Ring}.Result
    B=LinearInterpolate(Alpha=@n6) {constA=2, constB=0}
   ))
  n12 = Fmod(A=Multiply(A=Time() {period=1}, B=@n1) {constB=0.5}, B=Constant() {r=1})
  n13 = Power(Base=LinearInterpolate(A=@n12, B=OneMinus(None=@n12), Alpha=@n3)) {constExponent=1.1}
  n14 = LinearInterpolate(A=@n5, Alpha=@n13) {constA=0.1, constB=0.01}
  n15 = LinearInterpolate(Alpha=@n13) {constA=0.002, constB=0.05}
  n16 = LinearInterpolate(B=@n9, Alpha=@n13) {constB=0.5}
  n17 = Clamp(None=Multiply(
    A=MaterialFunctionCall(RingSize=@n14, EdgeSoftness=@n15, OuterRadius=@n16) {materialFunction=&MF_UI_Ring}.Result
    B=LinearInterpolate(Alpha=@n13) {constA=2, constB=0}
   ))
  NamedRerouteDeclaration(None=Multiply(A=Max(A=@n11, B=@n17), B=ScalarParameter() {defaultValue=1, group=Ping, parameterName=PingAmount, sliderMax=1})) {name=Pulse, nodeColor=<rgba 0.6641 1 0 1>}
 ]
}
